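# Sound cue for a looping solid-box ambience in the Chapter 5 Central Picnic environment actors.
# A looping wave player plays the box loop through an Enveloper whose cubic volume curve fades it in from near silence over a quarter second.
# It uses a dedicated attenuation asset for spatial falloff and is routed to the scripted events sound class.
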SoundCue /Game/Sounds/Helios/Environment_Actors/CentralPicnic/Cue_Ch5_CentralPicnic_Box_Solid
root=Enveloper_0
nodes [
 {
  id=Enveloper_0
  children [WavePlayer_1]
  properties {
   pitchCurve {editorCurveData {keys [{time=0}]}}
   volumeCurve {editorCurveData {keys [{interpMode=RCIM_Cubic, tangentMode=RCTM_User, time=0, value=0.001}, {interpMode=RCIM_Cubic, tangentMode=RCTM_User, time=0.25, value=1}]}}
  }
 }
 {
  id=WavePlayer_1
  properties {bLooping=true, soundWaveAssetPtr=&/CentralPicnic_Amb_Box_Solid_Loop_02}
 }
]
properties {
 attenuationSettings=&/ATT_Ch5_CentralPicnic_Box_Solid
 maxDistance=5500
 soundClassObject=/Game/Sounds/FortSoundClassesAndModes/Fort_Events_Scripted
 volumeMultiplier=1
}
